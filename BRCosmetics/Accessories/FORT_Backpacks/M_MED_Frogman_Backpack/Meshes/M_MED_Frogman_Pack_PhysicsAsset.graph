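# Physics asset for the Frogman backpack cosmetic, defining collision and joints for its dangling hoses, straps and knife.
# The pack body combines two spheres and a capsule, while the hoses use spheres and the straps and knife use boxes.
# Angular-limited constraints tie both hoses, the left strap and the knife to the pack.
# The right strap hangs off the knife instead, forming a two-link chain.
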PhysicsAsset /BRCosmetics/Accessories/FORT_Backpacks/M_MED_Frogman_Backpack/Meshes/M_MED_Frogman_Pack_PhysicsAsset
bodies {
 dyn_hose_l [{shapeType=Sphere, radius=2.8408, center=<9 0 0>}]
 dyn_hose_r [{shapeType=Sphere, radius=3.3225, center=<9 0 0>}]
 dyn_knife [{shapeType=Box, extentX=1.8318, extentY=7.7139, extentZ=24.0593, center=<2.8069 -0.1888 -0.3518>, rotation=<rot 0 -90.7568 -88.1543>}]
 dyn_strap_l [{shapeType=Box, extentX=3.8367, extentY=5.5287, extentZ=6, center=<3 0 0>, rotation=<rot 90 0 0>}]
 dyn_strap_r [{shapeType=Box, extentX=3.7226, extentY=5.5287, extentZ=6, center=<-3 0 0>, rotation=<rot 90 0 0>}]
 pack [
  {shapeType=Sphere, radius=9.6559, center=<-3.4449 -0.0001 -8.6854>}
  {shapeType=Sphere, radius=7.3688, center=<0.8596 -5.2867 -8.6854>}
  {shapeType=Capsule, length=16.2025, radius=6.9538, center=<-0.0559 -3.1981 -5.4367>, rotation=<rot 0.5893 -180 -180>}
 ]
}
constraints [
 {bone1Name=dyn_hose_l, bone2Name=pack, swing1LimitDegrees=12.5, swing2LimitDegrees=7.5, twistLimitDegrees=5}
 {bone1Name=dyn_hose_r, bone2Name=pack, swing1LimitDegrees=12.5, swing2LimitDegrees=7.5, twistLimitDegrees=5}
 {bone1Name=dyn_strap_l, bone2Name=pack, swing1LimitDegrees=22.5, swing2LimitDegrees=10, twistLimitDegrees=2}
 {bone1Name=dyn_strap_r, bone2Name=dyn_knife, swing1LimitDegrees=30, swing2LimitDegrees=19, twistLimitDegrees=5}
 {bone1Name=dyn_knife, bone2Name=pack, swing1LimitDegrees=10, swing2LimitDegrees=10, twistLimitDegrees=5}
]
